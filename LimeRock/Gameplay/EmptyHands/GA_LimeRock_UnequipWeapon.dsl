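; Gameplay ability for the LimeRock empty hands state that keeps the player's hands free of weapons while active.
; Where cosmetic events can run it hides the center reticle until the ability ends.
; Otherwise, it unequips the current weapon by inventory GUID and binds to the pawn's weapon-equipped delegate to undo any new equip.
; On end, that path unbinds and tries to re-activate the primary quickbar's focused weapon.
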(event Ability|EventActivateAbility
  (Networking|CanExecuteCosmeticEvents
    (:True
      (Class|BPLimeRockFunctionLibrary|HideHUDForLocalPlayer "(GameplayTags=((TagName=\"HUD.Center.Reticle\")))" true "GA_UnequipWeapon:: Removing reticle because of empty hands"))
    (:False
      (bind _asfort_player_pawn (Utilities|Casting|CastToFortPlayerPawn (Ability|GetAvatarActorfromActorInfo)))
      (bind _output_get (Variables|Default|SetActivatingPawn _asfort_player_pawn))
      (FortPawn|AssignOnWeaponEquippedDelegate _output_get (delegate Custom|OnWeaponEquippedDelegate_Event))
      (CallFunction|TryUnequipCurrentWeapon))))

(event Custom|OnWeaponEquippedDelegate_Event
  (params (NewWeapon :type Fort%20Weapon%20Object%20Reference) (PrevWeapon :type Fort%20Weapon%20Object%20Reference))
  ;[0.15 0.15 0.15 0.5 12] Intercept and unequip any new attempts to equip a weapon whilst empty hands is active
  (CallFunction|TryUnequipCurrentWeapon))

(event Ability|EventOnEndAbility
  (params (bWasCancelled :type Boolean))
  (bind _returnvalue (Fortnite|GetFortPlayerControllerfromActor (Variables|Default|GetActivatingPawn)))
  (bind _activatingpawn (Variables|Default|GetActivatingPawn))
  ;[0.15 0.15 0.15 0.5 12] If we had a weapon equipped when empty hands was given try and re-equip it when its over
  (Networking|CanExecuteCosmeticEvents
    (:True
      (Class|BPLimeRockFunctionLibrary|HideHUDForLocalPlayer "(GameplayTags=((TagName=\"HUD.Center.Reticle\")))" false "GA_UnequipWeapon:: Showing reticle because of empty hands"))
    (:False
      (Utilities|IsValid _activatingpawn
        (:Is%20Valid
          (FortPawn|UnbindEventfromOnWeaponEquippedDelegate _activatingpawn (delegate Custom|OnWeaponEquippedDelegate_Event))
          (Utilities|IsValid _returnvalue
            (:Is%20Valid
              (Class|FortPlayerController|TryActivatePrimaryQuickbarFocusedWeapon _returnvalue))))))))

(fn TryUnequipCurrentWeapon
  (bind _activatingpawn (Variables|Default|GetActivatingPawn))
  (bind _returnvalue (Weapon|GetCurrentWeapon _activatingpawn))
  (bind _returnvalue_2 (FortWeapon|GetInventoryGUID _returnvalue))
  (Utilities|IsValid _returnvalue
    (:Is%20Valid
      (Fortnite|FortShippingLog self (Utilities|String|Append "GA_Unequip::Unequiping player weapon: " (Guid|ToString(Guid) _returnvalue_2)))
      (Class|FortPlayerPawn|UnequipCurrentWeaponbyId (Variables|Default|GetActivatingPawn) _returnvalue_2 true))
    (:Is%20Not%20Valid
      (Fortnite|FortShippingLog self "GE_Unequip::Current weapon not valid" true))))
